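; Parent Blueprint for vehicle drift trail cosmetics, driving a Niagara trail from drift events.
; On apply, it assigns the system and feeds it a wheel offset (optionally per car body from a data table), bounds-derived wheel size, frontend preview flag and game mode ID.
; The trail starts on drift or landing mid-drift, gated by significance LOD, and stops when airborne or drift ends.
; Color-ready events set the paint parameters.
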(event Vehicle|Cosmetics|EventOnCosmeticApplied
  (bind _ns_drifttrail (Variables|Default|GetNS_DriftTrail))
  ;[1 1 1] Set Initial Values
  (Niagara|SetNiagaraSystemAsset _ns_drifttrail (Variables|Default|GetWheelTrailFX) true)
  (CallFunction|SetOffset)
  (CallFunction|SetWheelWrapSize)
  (bind _bisfrontend (CallFunction|SetupFrontEndPreview _ns_drifttrail))
  (CallFunction|DetectPlaylist _bisfrontend))

(event EventBeginPlay
  ;[1 1 1] Set Initial Values
  (Variables|/Script/Engine.SceneComponent|SetbAbsoluteRotation false (Variables|Default|GetNS_DriftTrail)))

(event Vehicle|Cosmetics|EventOnVehicleWheelsLeftGround
  ;[1 1 1] In Air(Jumping) Events
  (if (Components|Activation|IsActive (Variables|Default|GetNS_DriftTrail))
    (Components|Activation|Deactivate (Variables|Default|GetNS_DriftTrail))))

(event Vehicle|Cosmetics|EventOnDriftStarted
  (bind _vehiclecosmeticsinterface (Variables|Vehicle|Cosmetics|GetVehicleCosmeticsInterface))
  ;[1 1 1] Drifting Events
  (bind _returnvalue (VehicleCosmetics|FX|GetMaxSignificanceLODRearWheelDriftFX))
  (bind _btrigger (Class|BPVehicleCosmeticsLibrary|ShouldTriggerFX _vehiclecosmeticsinterface _returnvalue))
  (if _btrigger
    (if (or (Vehicle|Cosmetics|IsTouchingGroundforCosmetics _vehiclecosmeticsinterface) (Reactive|Frontend|IsCosmeticObjectDisplayedinFrontendContext 0))
      (Components|Activation|SetActive (Variables|Default|GetNS_DriftTrail) true false))))

(event Vehicle|Cosmetics|EventOnVehicleLanded
  (params (LandingForced :type Float%20%28single-precision%29) (bLandedKickflip :type Boolean))
  ;[1 1 1] In Air(Jumping) Events
  (if (Utilities|DoesObjectImplementInterface (Actor|GetOwner) "/Script/VehicleCosmeticsRuntime.VehicleCosmeticsInterface")
    (if (Vehicle|Cosmetics|Drifting|IsDrifting (Variables|Vehicle|Cosmetics|GetVehicleCosmeticsInterface))
      (bind _returnvalue (VehicleCosmetics|FX|GetMaxSignificanceLODRearWheelDriftFX))
      (bind _btrigger (Class|BPVehicleCosmeticsLibrary|ShouldTriggerFX (Variables|Vehicle|Cosmetics|GetVehicleCosmeticsInterface) _returnvalue))
      (if _btrigger
        (Components|Activation|Activate (Variables|Default|GetNS_DriftTrail) false)))))

(event Vehicle|Cosmetics|EventOnDriftFinished
  ;[1 1 1] Drifting Events
  (Components|Activation|SetActive (Variables|Default|GetNS_DriftTrail) false false))

(event Vehicle|Cosmetics|EventOnColorReady
  (params (Color :type Name :default "None") (AdditionalInfo :type Vehicle%20Cosmetics%20Additional%20Parameter%20Infos%20Structure))
  ;[1 1 1] Setup Paint Settings
  (CallFunction|SetupPaintSettings Color AdditionalInfo))

(fn DetectPlaylist
  (params (isFrontEnd :type Boolean))
  (bind _returnvalue (Fortnite|Athena|GameState|GetGameStateAthena))
  (bind _ns_drifttrail (Variables|Default|GetNS_DriftTrail))
  (if isFrontEnd
    (if (Reactive|Frontend|IsCosmeticObjectDisplayedinLocker (Variables|Vehicle|Cosmetics|GetVehicleCosmeticComponent))
      (LocalVariables|SetGameMode 0)
      (else
        (LocalVariables|SetGameMode 4)))
    (else
      (Utilities|IsValid _returnvalue
        (:Is%20Valid
          (if (Class|FortGameStateAthena|IsBattleRoyalePlaylist _returnvalue true)
            (LocalVariables|SetGameMode 1)
            (elif (Utilities|String|Contains (Utilities|GetDisplayName (Game|GetGameMode)) "DelMar")
              (LocalVariables|SetGameMode 2)
              (else
                (LocalVariables|SetGameMode 3))))
          (Niagara|SetNiagaraVariableByString(Int32) _ns_drifttrail "GameModeID" (LocalVariables|GetGameMode))
          (Niagara|SetNiagaraVariableByString(Bool) _ns_drifttrail "bPreviewVehicleSpeed" false))
        (:Is%20Not%20Valid
          _)))))

(fn SetOffset
  (bind _vehiclecosmeticsinterface (Variables|Vehicle|Cosmetics|GetVehicleCosmeticsInterface))
  (bind _returnvalue (Vehicle|Cosmetics|GetBodyItemName _vehiclecosmeticsinterface))
  (if (Variables|Default|GetUseOffsetPerCar?)
    (bind _returnvalue_2 (Utilities|GetDataTableRow "/VehicleCosmetics/DataTables/DT_VehicleCosmetics_TurboOffsets.DT_VehicleCosmetics_TurboOffsets" _returnvalue)
      (:then
        (Variables|Default|SetOffsetFromWheels (+ (Variables|Default|GetOffsetFromWheels) (.DriftTrailOffset_6_7E41BD0641CAC23B68A086AFDC86F1BF (Utilities|Struct|BreakStructVehiclesCosmeticsTurboOffset _returnvalue_2)))))
      (:RowNotFound
        (Development|PrintString (Utilities|String|Append "CarBody is not in the table for DriftTrail Offsets - " (Utilities|String|ToString(Name) _returnvalue)) true true "(R=0.937356,G=1.000000,B=0.000000,A=1.000000)" 5.0))))
  (Niagara|SetNiagaraVariable(Vector3) (Variables|Default|GetNS_DriftTrail) "offset" (Variables|Default|GetOffsetFromWheels))
  (return))

(fn SetupFrontEndPreview
  (params (NS_Drift :type Niagara%20Particle%20System%20Component%20Object%20Reference))
  (bind _vehiclecosmeticcomponent (Variables|Vehicle|Cosmetics|GetVehicleCosmeticComponent))
  (bind _returnvalue (Reactive|Frontend|IsCosmeticObjectDisplayedinFrontendContext _vehiclecosmeticcomponent))
  (Niagara|SetNiagaraVariable(Bool) NS_Drift "User.bIsFrontendPreview" _returnvalue)
  (return _returnvalue))

(fn SetupPaintSettings
  (params (Color :type Name) (AdditionalInfo :type Vehicle%20Cosmetics%20Additional%20Parameter%20Infos%20Structure))
  (bind _ns_drifttrail (Variables|Default|GetNS_DriftTrail))
  (if (Utilities|String|EqualExactly(String) (Utilities|String|ToString(Name) Color) "None")
    (Niagara|SetNiagaraVariable(Bool) _ns_drifttrail "bIsPainted" false)
    (else
      (Niagara|SetNiagaraVariable(Bool) _ns_drifttrail "bIsPainted" true)
      (Niagara|SetNiagaraVariable(Float) (Variables|Default|GetNS_DriftTrail) "PaintedScalar" (Variables|Default|GetPaintedScalar))
      (bind _color_index (Class|BPVehicleCosmeticsFunctions|BPGetNSIndexbyColor (Variables|GetColor)))
      (Niagara|SetNiagaraVariable(Int32) (Variables|Default|GetNS_DriftTrail) "ColorVariantID" _color_index)
      (Class|BPVehicleCosmeticsFunctionLibrary|CosmeticActorColorReady (Variables|GetColor) "PaintedColor" (Variables|GetAdditionalInfo) (Variables|Default|GetNS_DriftTrail))
      (bind _returnvalue (Utilities|GetDataTableRow "/VehicleCosmetics/DataTables/DT_VehicleCosmetics_Painted.DT_VehicleCosmetics_Painted" (Variables|GetColor)))
      (return (Utilities|Struct|BreakStructVehicleCosmeticsPainted _returnvalue) (.Balanced_34_F677FC9D4B1B4146E8CC9DBCB9E0943E (Utilities|Struct|BreakStructVehicleCosmeticsPainted _returnvalue)) (.Metallic_41_F57CE68E419E301DD160BEB5D0E32BFC (Utilities|Struct|BreakStructVehicleCosmeticsPainted _returnvalue))))))

(fn SetWheelWrapSize
  (bind _result (ArrayMacros|SafeArrayGet (Vehicle|Cosmetics|GetWheelSkeletalMeshComponents (Variables|Vehicle|Cosmetics|GetVehicleCosmeticsInterface)) 3)
    (:%20%20
      (bind (_returnvalue_x _returnvalue_y _returnvalue_z) (Math|Vector|BreakVector (/ (Math|Vector|Normalize (.BoxExtent (Collision|GetComponentBounds _result))) (Math|Conversions|ToVector(Float) 15.0))))
      (Niagara|SetNiagaraVariable(Float) (Variables|Default|GetNS_DriftTrail) "WheelSize" _returnvalue_z))))

(fn ConstructionScript)
